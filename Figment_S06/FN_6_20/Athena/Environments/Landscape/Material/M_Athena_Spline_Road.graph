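# Landscape spline road material for the Athena environment that mixes a grasslands terrain texture with a street road texture.
# A HeightLerp blends the two diffuse maps, with vertex color red as transition phase and inverted grass alpha as height.
# A Contrast parameter controls blend sharpness.
# A grasslands normal map is sampled, but the graph shows no connection from it or the blend to material outputs.
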Material /Figment_S06/FN_6_20/Athena/Environments/Landscape/Material/M_Athena_Spline_Road
graph {
 properties {pixelDepthOffsetMode=PDOM_Legacy, usedShadingModels=Default Lit}
 nodes [
  n1 = TextureSample() {texture=/Figment_S06/FN_6_20/Packages/DS_Fortnite_Terrain_NoLOD/Textures/T_Grasslands_D}
  MaterialFunctionCall(
   A=@n1.RGB
   B=TextureSample() {texture=/FigmentEnvironment/FN_6_20/Packages/DS_Fortnight/SM/Textures/T_Street_Road_D}.RGB
   Transition Phase=VertexColor().R
   Height Texture=OneMinus(None=@n1.A)
   Contrast=ScalarParameter() {defaultValue=5, parameterName=Contrast}
  ) {materialFunction=&HeightLerp}
  TextureSample() {samplerType=SAMPLERTYPE_Normal, texture=/FigmentEnvironment/FN_6_20/Packages/DS_Fortnite_Terrain_NoLOD/Textures/T_Grasslands_N}
 ]
}
